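# Master material for the BlueSteel SE Infinity vehicle wheel, adding a color-customizable infinity glow to rim and tire.
# Channel-packed rim and tire masks drive emissive, with MF_Infinity_OffsetFunction supplying bump-offset depth in the green region.
# Static switches add panning swirl emissive for painted variants and tiled detail normals for rubber.
# Emissive, normal, roughness and metallic override MF_WheelParent_Simple, then pass through vehicle damage and de-rez transition functions.
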Material /VehicleCosmetics/Wheels/BlueSteelSE_Infinity/Materials/M_Wheel_BlueSteel_SE_Infinity
graph {
 properties {
  bAutomaticallySetUsageInEditor=false
  bUseMaterialAttributes=true
  bUsedWithMeshDeformer=true
  bUsedWithSkeletalMesh=true
  pixelDepthOffsetMode=PDOM_Legacy
  previewMesh=/VehicleCosmetics/Wheels/BlueSteelSE_Infinity/Meshes/SK_Bluesteel_SE_Infinity
  usedShadingModels=Default Lit
 }
 nodes [
  n1 = MaterialFunctionCall() {materialFunction=&MF_WheelParent_Simple}
  n2 = NamedRerouteUsage()
  n3 = Constant() {r=2}
  n4 = NamedRerouteUsage()
  n5 = Constant() {r=1}
  n6 = SetMaterialAttributes!(
   MaterialAttributes=@n1.MaterialAttributes
   Emissive Color=NamedRerouteUsage()
   Normal=LinearInterpolate(A=NamedRerouteUsage(), B=NamedRerouteUsage(), Alpha=NamedRerouteUsage())
   Roughness=LinearInterpolate(A=NamedRerouteUsage(), B=Multiply(A=@n2, B=@n3), Alpha=@n2)
   Metallic=LinearInterpolate(A=NamedRerouteUsage(), B=Multiply(A=@n4, B=@n5), Alpha=@n4)
  ) {attributeSetTypes [
    #1
    #2
    #3
    #4
   ]}
  MaterialFunctionCall(In=MaterialFunctionCall(Material=@n6) {materialFunction=&MF_BaseVehicle_Damage}.Result) {materialFunction=&MF_VehicleDeRez_Transition}
  n7 = Constant() {r=0.15}
  n8 = Multiply(A=VectorParameter() {group=Infinity, parameterName=CustomColor, defaultValue=<rgba 0.0176 0.521 1 1>}.RGB, B=Constant() {r=2.5})
  n9 = Panner(Coordinate=TextureCoordinate() {uTiling=4, vTiling=0.5}) {speedY=-0.1}
  n10 = Panner(Coordinate=TextureCoordinate() {uTiling=2, vTiling=0.5}) {speedY=0.011}
  n11 = Power(
   Base=Multiply(
    A=TextureSample(UVs=@n9) {texture=/VehicleCosmetics/Wheels/Uncle/Textures/Swirls02_D}.RGB
    B=TextureSample(UVs=@n10) {texture=&Swirls02_D}.RGB
   )
   Exp=Constant() {r=1.5}
  )
  n12 = Constant() {r=25}
  n13 = Multiply(A=LinearInterpolate(A=Multiply(A=@n7, B=@n8), B=@n8, Alpha=Multiply(A=@n11, B=@n12)), B=Add(A=NamedRerouteUsage(), B=NamedRerouteUsage()))
  n14 = VectorParameter() {group=Infinity, parameterName=CustomColor, defaultValue=<rgba 0.0176 0.521 1 1>}
  n15 = TextureSampleParameter2D() {
   group=Infinity
   parameterName=RimEmissiveMask
   samplerType=SAMPLERTYPE_LinearColor
   texture=/VehicleCosmetics/Wheels/BlueSteelSE_Infinity/Textures/T_Bluesteel_SE_Infinity_Rim_M
  }
  n16 = TextureSampleParameter2D() {
   group=Infinity
   parameterName=TireEmissiveMask
   samplerType=SAMPLERTYPE_LinearColor
   texture=/VehicleCosmetics/Wheels/BlueSteelSE_Infinity/Textures/T_Bluesteel_SE_Infinity_Tire_M_Emission
  }
  n17 = LinearInterpolate(A=@n15.RGB, B=@n16.RGB, Alpha=NamedRerouteUsage())
  n18 = NamedRerouteDeclaration(None=ComponentMask(None=@n17) {r=true}) {name=Mask_R, nodeColor=<rgba 0 1 0.3203 1>}
  n19 = ScalarParameter() {defaultValue=3, group=Infinity, parameterName=InfinityGlowBrightness}
  n20 = NamedRerouteDeclaration(None=ComponentMask(None=@n17) {b=true}) {name=Mask_B, nodeColor=<rgba 0.6406 0 1 1>}
  n21 = TextureObjectParameter() {
   group=Infinity
   parameterName=RimInfMask
   samplerType=SAMPLERTYPE_LinearColor
   sortPriority=33
   texture=/VehicleCosmetics/Wheels/BlueSteelSE_Infinity/Textures/T_Bluesteel_SE_Infinity_Rim_Inf
  }
  n22 = MaterialFunctionCall!(
   Texture=@n21
   Offset=ScalarParameter() {defaultValue=0.01, group=Infinity, parameterName=BumpOffsetAmount}
   UVs
   RotateUVs
   GradientFalloff
   GradientDarkness=ScalarParameter() {defaultValue=0.01, group=Infinity, parameterName=GradientDarkness}
  ) {materialFunction=/VehicleCosmetics/Wheels/ClockGear_SE/Materials/MF_Infinity_OffsetFunction}
  n23 = TextureObjectParameter() {
   group=Infinity
   parameterName=TireInfMask
   sortPriority=33
   texture=/VehicleCosmetics/Wheels/BlueSteelSE_Infinity/Textures/Wheel_BlueSteel_Infinity_SE_Inf
  }
  n24 = MaterialFunctionCall(
   Texture=@n23
   Offset=ScalarParameter() {defaultValue=0.01, group=Infinity, parameterName=BumpOffsetAmount}
   GradientDarkness=ScalarParameter() {defaultValue=0.01, group=Infinity, parameterName=GradientDarkness}
  ) {materialFunction=&MF_Infinity_OffsetFunction}
  n25 = NamedRerouteUsage()
  n26 = Multiply(
   A=ComponentMask(None=LinearInterpolate(A=@n22.Result, B=@n24.Result, Alpha=@n25)) {g=true}
   B=NamedRerouteDeclaration(None=ComponentMask(None=@n17) {g=true}) {name=Mask_G, nodeColor=<rgba 1 0.9609 0 1>}
  )
  n27 = Add(
   A=Multiply(A=Multiply(A=@n14.RGB, B=@n18), B=ScalarParameter() {defaultValue=2, group=Infinity, parameterName=EmissiveGlow})
   B=Multiply(A=Multiply(A=@n14.RGB, B=@n19), B=Add(A=@n20, B=@n26))
  )
  n28 = NamedRerouteUsage()
  NamedRerouteDeclaration(None=StaticSwitchParameter(True=LinearInterpolate(A=@n13, B=@n27, Alpha=@n28), False=@n27) {group=Infinity, parameterName=Painted}) {name=EmissiveColor, nodeColor=<rgba 0 0.4554 1 1>}
  NamedRerouteDeclaration(None=Multiply(A=NamedRerouteUsage(), B=ScalarParameter() {defaultValue=1})) {name=Metallic, nodeColor=<rgba 0.7109 1 0 1>}
  n29 = GetMaterialAttributes(None=@n1.MaterialAttributes) {attributeGetTypes [#2, #4, #3]}
  NamedRerouteDeclaration(None=@n29.Metallic) {name=Metallic IN, nodeColor=<rgba 0.4062 0 1 1>}
  NamedRerouteDeclaration(None=@n29.Roughness) {name=Roughness, nodeColor=<rgba 0.5811 1 0.029 1>}
  NamedRerouteDeclaration(None=VertexColor().R) {name=VC_R, nodeColor=<rgba 1 0.0906 0 1>}
  NamedRerouteDeclaration(None=@n29.Normal) {name=Normal IN, nodeColor=<rgba 1 0.4219 0 1>}
  n30 = NamedRerouteUsage()
  n31 = TextureCoordinate()
  n32 = ScalarParameter() {defaultValue=8, group=TireRubber, parameterName=TextureDetailNormalTiling}
  n33 = Multiply(
   A=TextureSample(UVs=Multiply(A=@n31, B=@n32)) {samplerType=SAMPLERTYPE_Normal, texture=/VehicleCosmetics/Wheels/BlueSteelSE_Infinity/Textures/Asphalt_N}.RGB
   B=OneMinus(None=NamedRerouteUsage())
  )
  NamedRerouteDeclaration(None=StaticSwitchParameter(True=Add(A=@n30, B=@n33), False=@n30) {group=TireRubber, parameterName=TireRubberExist}) {name=NormalRubber, nodeColor=<rgba 0.1016 0 1 1>}
 ]
}
